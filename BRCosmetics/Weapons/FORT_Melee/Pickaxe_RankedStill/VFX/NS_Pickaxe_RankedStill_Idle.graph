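# Niagara idle glow effect for the RankedStill pickaxe, with a Glow emitter that lines the weapon with soft sprites.
# One burst of 24 particles is partitioned so each takes a fixed size and socket-relative offset from arrays.
# Sprites are camera-offset, fade in to low opacity by system age and take a system-level color.
# User parameters expose ranked progression, which also sets a GlowBool via integer comparison, and a ranked-colors toggle.
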NiagaraSystem /BRCosmetics/Weapons/FORT_Melee/Pickaxe_RankedStill/VFX/NS_Pickaxe_RankedStill_Idle
dependencies {
 usedDataInterfaces [
  &NiagaraDataInterfaceArrayFloat
  &NiagaraDataInterfaceArrayFloat3
  &NiagaraDataInterfaceCurve
  &NiagaraDataInterfaceSocketReader
  &NiagaraDataInterfaceArrayInt32
  &NiagaraDataInterfaceArrayColor
 ]
 usedDynamicInputs [
  &SelectIntFromArray
  &MakeCustomLinearColorFromBool
  &SelectLinearColorFromArray
 ]
 usedModules [
  &self:SystemSpawnScript.SystemScriptSource.SystemScriptGraph.NiagaraNodeAssignment_0.SetVariables_#1
  &SystemState
  &self:SystemSpawnScript.SystemScriptSource.SystemScriptGraph.NiagaraNodeAssignment_2.SetVariables_#2
  &self:SystemSpawnScript.SystemScriptSource.SystemScriptGraph.NiagaraNodeAssignment_1.SetVariables_#3
 ]
}
emitters {Glow {
  emitterUpdateScript [
   &EmitterState(
    Inactive Response=::NewEnumerator1
    Life Cycle Mode=::NewEnumerator1
    Loop Behavior=::NewEnumerator1
    Loop Duration Mode=::NewEnumerator1
    UseLoopDelay=false
   )
   &SpawnBurst_Instantaneous(Spawn Count=24, Use Spawn Probability=false, UseLoopCountLimit=true)
   /SetVariables(Emitter.GlowBool:NiagaraBool=&SetBoolByIntComparison(A=$User.RankedProgression, B=15, Comparison Type=::NewEnumerator0)) {isSetParameters=true}
  ]
  particleSpawnScript [
   &PartitionParticles(Number Of Partitions=24)
   &InitializeParticle(
    Sprite Size Mode=::NewEnumerator3
    Uniform Sprite Size=&SelectFloatFromArray(
     Direct Array Index=$Particles.PartitionParticles.Partition
     Float Selection Array=DataInterface() {propertyValues {FloatData [60,60,60,60,60,60,60,60,80,70,60,60,60,45,45,45,45,45,45,45,45,45,45,70]}}
    )
    UsePositionOffset=false
    Write Material Random=true
   )
   &SocketLocation(
    SamplePosOffset=true
    Sampled Position Offset=&SelectVectorFromArray(
     Direct Array Index=$Particles.PartitionParticles.Partition
     Vector Selection Array=DataInterface() {propertyValues {FloatData [
        <7 7 -51>
        <7 -7 -51>
        <-7 7 -51>
        <-7 -7 -51>
        <7 0 -51>
        <-7 0 -51>
        <0 7 -51>
        <0 -7 -51>
        <0 0 -35>
        <0 0 -30>
        <0 0 -20>
        <0 0 -10>
        <0 0 -5>
        <0 0 -55>
        <0 0 -62>
        <0 0 -72>
        <0 0 -82>
        <0 0 -92>
        <0 0 -102>
        <0 0 -112>
        <0 0 -122>
        <0 0 -132>
        <0 0 -142>
        <0 0 -150>
       ]}}
    )
    Socket Reader=$System.Socket Reader
    Socket Sampling Mode=::NewEnumerator2
    VelocitySpeedLimit=false
   )
   &CameraOffset(Camera Offset Amount=-25)
  ]
  particleUpdateScript [
   &ParticleState(Kill Particles When Lifetime Has Elapsed=false, Let Infinitely Lived Particles Die When Emitter Deactivates=true)
   &ScaleColor(
    Scale Alpha=&Multiply_Float(
     A=0.125
     B=&FloatFromCurve(
      FloatCurve=DataInterface() {propertyValues {Curve {defaultValue=1, keys [{tangentMode=RCTM_User, time=0, value=0}, {tangentMode=RCTM_User, time=0.1, value=1}]}}}
      CurveIndex=$Engine.System.Age
     )
    )
    Scale RGB=&MakeVectorFromLinearColorRGB(Color=$System.ColorArray1)
    ScaleA=true
    ScaleRGB=true
   )
  ]
  rendererClasses [&NiagaraSpriteRendererProperties]
  renderers [{rendererClass=&NiagaraSpriteRendererProperties}]
 }}
summary {
 systemName=NS_Pickaxe_RankedStill_Idle
 emitters [{emitterName=Glow, rendererClasses [&NiagaraSpriteRendererProperties]}]
 userVariables [
  User.RankedProgression {type=NiagaraInt32}
  User.UseRankedColors {type=NiagaraBool}
 ]
}
