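# Niagara swing trail for the RankedStill pickaxe, exposing user inputs for ranked progression and ranked colors.
# Two emitters spawn short-lived anim-trail ribbons between two bones or sockets, offset by a ribbon ID utility.
# MainTrails feeds three renderers, tinted by system color and time-of-day values, with a separate soft color and narrower width, fading with the swing.
# SupportingTrails adds a black ribbon whose fade uses time-of-day scalars.
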NiagaraSystem /BRCosmetics/Weapons/FORT_Melee/Pickaxe_RankedStill/VFX/NS_Pickaxe_RankedStill_Trail
dependencies {
 usedDataInterfaces [
  &NiagaraDataInterfaceCurve
  &NiagaraDataInterfaceSocketReader
  &NiagaraDataInterfaceArrayInt32
  &NiagaraDataInterfaceArrayColor
  &NiagaraDataInterfaceColorCurve
 ]
 usedDynamicInputs [
  &SelectIntFromArray
  &MakeCustomLinearColorFromBool
  &SelectLinearColorFromArray
  &ColorFromCurve
 ]
 usedModules [
  &self:SystemSpawnScript.SystemScriptSource.SystemScriptGraph.NiagaraNodeAssignment_3.SetVariables_#1
  &SystemState
  &self:SystemSpawnScript.SystemScriptSource.SystemScriptGraph.NiagaraNodeAssignment_0.SetVariables_#2
  &self:SystemSpawnScript.SystemScriptSource.SystemScriptGraph.NiagaraNodeAssignment_4.SetVariables_#3
  &self:SystemSpawnScript.SystemScriptSource.SystemScriptGraph.NiagaraNodeAssignment_1.SetVariables_#4
  &self:SystemSpawnScript.SystemScriptSource.SystemScriptGraph.NiagaraNodeAssignment_2.SetVariables_#5
 ]
}
emitters {
 MainTrails {
  emitterSpawnScript [&FN_RibbonIDOffsetUtility()]
  emitterUpdateScript [
   &EmitterState()
   &FN_SpawnAnimTrail()
  ]
  particleSpawnScript [
   &FN_InitializeAnimTrailBoneIndecies(First Bone or Socket Index=1, Second Bone or Socket Index=0)
   &FN_InitializeAnimTrail(
    Color=&MultiplyLinearColor(Linear Color=$System.ColorArray1, Scale Factor=&MakeVector4_FromLinearColor(LinearColor=$System.TODColor))
    Lifetime=0.13
   )
  ]
  particleUpdateScript [
   &FN_UpdateAnimTrail()
   &ScaleColor(
    Scale Alpha=&Multiply_Float(
     A=&FloatFromCurve(
      FloatCurve=DataInterface() {propertyValues {Curve {keys [{interpMode=RCIM_Cubic, leaveTangent=-4.0138, tangentMode=RCTM_Break, time=0, value=1}, <at 0.4 0>]}}}
      Scale Curve=$System.SwingGeneralFade
     )
     B=0.5
    )
    Scale RGB=&MakeVectorFromLinearColorRGB(Color=<rgba 1 1 1 1>)
    ScaleA=true
    ScaleRGB=true
   )
   &ScaleColor(Scale RGB=&VectorFromFloat(Value=$System.TODFloat), ScaleA=false, ScaleRGB=true) {module=ScaleColor001}
   /SetVariables(Particles.SoftColor:LinearColor=&MakeLinearColorFromVectorAndFloat(
     "Vector (RGB)"=&Multiply_Vector(
      A=&MakeVectorFromLinearColorRGB(Color=$System.ColorArray1)
      B=&MakeVectorFromLinearColorRGB(Color=$System.TODColor)
     )
     "Float (Alpha)"=&Multiply_Float(
      A=&FloatFromCurve(
       FloatCurve=DataInterface() {propertyValues {Curve {keys [{interpMode=RCIM_Cubic, leaveTangent=-4.0138, tangentMode=RCTM_Break, time=0, value=1}, <at 0.4 0>]}}}
       Scale Curve=$System.SwingGeneralFade
      )
      B=1
     )
    )) {isSetParameters=true}
   /SetVariables(Particles.SoftRibbonWidth:NiagaraFloat=&Multiply_Float(A=$Particles.Initial.RibbonWidth, B=0.8)) {isSetParameters=true}
  ]
  rendererClasses [&NiagaraRibbonRendererProperties]
  renderers [
   {rendererClass=&NiagaraRibbonRendererProperties}
   {rendererClass=&NiagaraRibbonRendererProperties, rendererIndex=1}
   {rendererClass=&NiagaraRibbonRendererProperties, rendererIndex=2}
  ]
 }
 SupportingTrails {
  emitterSpawnScript [&FN_RibbonIDOffsetUtility()]
  emitterUpdateScript [
   &EmitterState()
   &FN_SpawnAnimTrail(Spawn Interval=0)
  ]
  particleSpawnScript [
   &FN_InitializeAnimTrailBoneIndecies(First Bone or Socket Index=1, Second Bone or Socket Index=0)
   &FN_InitializeAnimTrail(Color=<rgba 0 0 0 1>, Lifetime=0.1)
  ]
  particleUpdateScript [
   &SolveForcesAndVelocity(Clamp Velocity=false, Limit Acceleration=false)
   &FN_UpdateAnimTrail(Locked Particle RibbonWidth=1) {module=FN_UpdateAnimTrail001}
   &ScaleColor(
    Scale Alpha=&FloatFromCurve(
     FloatCurve=DataInterface() {propertyValues {Curve {keys [
         {interpMode=RCIM_Cubic, tangentMode=RCTM_User, time=0, value=1}
         {arriveTangent=-5.0049, interpMode=RCIM_Cubic, leaveTangent=-5.0049, tangentMode=RCTM_User, time=0.5, value=0}
        ]}}}
     Scale Curve=&FNTimeOfDayScalar(Night Scale=1, Morning Scale=1, Evening Scale=1)
    )
    Scale RGB=&VectorFromFloat(Value=&FNTimeOfDayScalar(Night Scale=0.5, Morning Scale=0.7, Evening Scale=0.7))
    ScaleA=true
    ScaleRGB=true
   )
  ]
  rendererClasses [&NiagaraRibbonRendererProperties]
  renderers [{rendererClass=&NiagaraRibbonRendererProperties}]
 }
}
summary {
 systemName=NS_Pickaxe_RankedStill_Trail
 emitters [
  {emitterName=SupportingTrails, rendererClasses [&NiagaraRibbonRendererProperties]}
  {emitterName=MainTrails, rendererClasses [&NiagaraRibbonRendererProperties]}
 ]
 userVariables [
  User.RankedProgression {type=NiagaraInt32}
  User.RibbonIDOffset {type=NiagaraInt32}
  User.UseRankedColors {type=NiagaraBool}
 ]
}
